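# Physics asset for the medium male Director player body, covering dynamic accessory bones (holster, clip, stabler) around the hips rather than a full ragdoll.
# Capsules on pelvis, lower spine and thighs form the body shapes, joined to the pelvis by free angular constraints.
# The holster hangs from the pelvis with free twist, while the clip and three-part stabler chains start on free pelvis joints.
# Their child links tighten: Clip2 only twists, and the stabler narrows toward its tip.
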PhysicsAsset /Game/Characters/Player/Male/Medium/Bodies/M_MED_Director/Meshes/M_MED_Director_Physics
bodies {
 dyn_Clip [{shapeType=Capsule, length=6.1404, radius=1.2342, center=<-0.0454 -2.0152 0.1441>, rotation=<rot 0 -89.9995 88.7818>}]
 dyn_Clip2 [{shapeType=Box, extentX=7.661, extentY=1.3364, extentZ=4.4299, center=<-0.2036 0.8009 2.318>, rotation=<rot 3.3694 0.9085 15.0999>}]
 dyn_holster [{shapeType=Box, extentX=23.3569, extentY=33.9025, extentZ=4.0583, center=<4.8786 -0.0002 -3.6393>, rotation=<rot -5.6249 0 0>}]
 dyn_stabler1 [
  {shapeType=Box, extentX=1.51, extentY=2.1975, extentZ=1.01}
  {shapeType=Box, extentX=1.1875, extentY=12, extentZ=12.8125, center=<7.6024 3.2253 -0.9828>, rotation=<rot 0 -25.312 8.4387>}
 ]
 dyn_stabler2 [{shapeType=Box, extentX=1.3125, extentY=11.7192, extentZ=3.1905, center=<-0.8118 7.1227 -0.869>, rotation=<rot -0.8241 11.2179 5.5663>}]
 dyn_stabler3 [{shapeType=Box, extentX=2.2875, extentY=3.7881, extentZ=13.054, center=<1.7298 6.2673 -0.7362>, rotation=<rot -1.0004 -4.0652 -83.2388>}]
 pelvis [{shapeType=Capsule, length=1.2019, radius=12.2156, center=<0 0 0.5432>}]
 spine_01 [{shapeType=Capsule, length=12.8847, radius=6.8268}]
 thigh_l [{shapeType=Capsule, length=12.5055, radius=7.6396, center=<-8.0727 -0.0001 0.9174>, rotation=<rot 90 0 0>}]
 thigh_r [{shapeType=Capsule, length=3.9384, radius=6.7083, center=<3.078 -2.5783 -0.8385>, rotation=<rot -0.9287 12.9563 -89.049>}]
}
constraints [
 {bone1Name=dyn_holster, bone2Name=pelvis, swing1LimitDegrees=90, swing2LimitDegrees=35, twistLimitDegrees=45, twistMotion=Free}
 {
  bone1Name=thigh_r
  bone2Name=pelvis
  swing1LimitDegrees=45
  swing1Motion=Free
  swing2LimitDegrees=45
  swing2Motion=Free
  twistLimitDegrees=45
  twistMotion=Free
 }
 {
  bone1Name=thigh_l
  bone2Name=pelvis
  swing1LimitDegrees=45
  swing1Motion=Free
  swing2LimitDegrees=45
  swing2Motion=Free
  twistLimitDegrees=45
  twistMotion=Free
 }
 {
  bone1Name=spine_01
  bone2Name=pelvis
  swing1LimitDegrees=45
  swing1Motion=Free
  swing2LimitDegrees=45
  swing2Motion=Free
  twistLimitDegrees=45
  twistMotion=Free
 }
 {
  bone1Name=dyn_stabler1
  bone2Name=pelvis
  swing1LimitDegrees=45
  swing1Motion=Free
  swing2LimitDegrees=45
  swing2Motion=Free
  twistLimitDegrees=45
  twistMotion=Free
 }
 {bone1Name=dyn_stabler2, bone2Name=dyn_stabler1, swing1LimitDegrees=33, swing2LimitDegrees=8, twistLimitDegrees=30}
 {
  bone1Name=dyn_Clip
  bone2Name=pelvis
  swing1LimitDegrees=45
  swing1Motion=Free
  swing2LimitDegrees=45
  swing2Motion=Free
  twistLimitDegrees=45
  twistMotion=Free
 }
 {
  bone1Name=dyn_Clip2
  bone2Name=dyn_Clip
  swing1LimitDegrees=1
  swing1Motion=Locked
  swing2LimitDegrees=1
  swing2Motion=Locked
  twistLimitDegrees=105
 }
 {bone1Name=dyn_stabler3, bone2Name=dyn_stabler2, swing1LimitDegrees=10, swing2LimitDegrees=0, twistLimitDegrees=0}
]
